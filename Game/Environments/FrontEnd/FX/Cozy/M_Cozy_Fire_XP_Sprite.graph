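# Translucent unlit ember sprite material for the Cozy FrontEnd fire XP effect, usable with Niagara and Cascade particles.
# An ember texture's color and alpha are each multiplied by particle color and alpha.
# Two channels of a random-dot noise texture, tiled, rotated and panned at separate speeds (zero by default), sum into a mask.
# That mask blends from black to an orange tint, which is added to the ember color.
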Material /Game/Environments/FrontEnd/FX/Cozy/M_Cozy_Fire_XP_Sprite
graph {
 properties {
  bUsedWithNiagaraSprites=true
  bUsedWithParticleSprites=true
  blendMode=BLEND_Translucent
  numCustomizedUVs=1
  pixelDepthOffsetMode=PDOM_Legacy
  shadingModel=MSM_Unlit
  usedShadingModels=Unlit
 }
 nodes [
  n1 = MaterialFunctionCall(
   UVs=Multiply(A=TextureCoordinate(), B=ScalarParameter() {defaultValue=10, parameterName=DotSize})
   "Rotation Angle (0-1)"=ScalarParameter() {defaultValue=0.135, parameterName=Rotation}
  ) {materialFunction=&CustomRotator}
  n2 = ScalarParameter() {parameterName=dot1Speed}
  n3 = ScalarParameter() {parameterName=dot2Speed}
  n4 = Add(
   A=TextureSample(UVs=Panner(Coordinate=@n1.Rotated Values, Speed=@n2)) {texture=/Game/Global/Textures/Default/Noise/T_LockDeviceRandomDots}.R
   B=TextureSample(UVs=Panner(Coordinate=@n1.Rotated Values, Speed=@n3)) {texture=&T_LockDeviceRandomDots}.G
  )
  n5 = TextureSampleParameter2D() {
   parameterName=Texture
   samplerSource=SSM_Clamp_WorldGroupSettings
   texture=/Game/Environments/FrontEnd/FX/Cozy/Textures/T_WinterRestEmber_1000
  }
  Add(A=@n4, B=@n5.A)
  n6 = ParticleColor()
  Add(
   A=Multiply(A=@n5.RGB, B=@n6.RGB)
   B=LinearInterpolate(B=VectorParameter() {parameterName=Dot1Color, defaultValue=<rgba 1 0.0976 0 1>}.RGB, Alpha=@n4)
  )
  MaterialFunctionCall(UVs=TextureCoordinate(), Texture Scale=VectorParameter() {parameterName=TextureScale, defaultValue=<rgba 1 0.8 0 1>}.RGB) {materialFunction=&ScaleUVsByCenter}
  Multiply(A=@n5.A, B=@n6.A)
 ]
}
