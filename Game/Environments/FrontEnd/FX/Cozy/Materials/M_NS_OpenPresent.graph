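# Two-sided lit Niagara mesh-particle material for a cozy front-end present-opening effect, using a holiday wrapping-paper texture.
# The texture's green channel blends particle color with a hue-shifted copy, scaled by an emissive multiplier.
# A time-panned sine wave along object-relative world Z builds an offset along vertex normals.
# Dynamic parameters drive texture tiling, wave strength, tiling and speed, hue shift and emissive, while scalar parameters set wave hardness and intensity.
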Material /Game/Environments/FrontEnd/FX/Cozy/Materials/M_NS_OpenPresent
graph {
 properties {bUsedWithNiagaraMeshParticles=true, pixelDepthOffsetMode=PDOM_Legacy, twoSided=true, usedShadingModels=Default Lit}
 nodes [
  Constant3Vector() {constant=<rgba 1 0 0 1>}
  Constant() {r=1}
  Constant() {r=1}
  Constant() {r=0.1}
  n1 = WorldPosition()
  n2 = ObjectPositionWS()
  n3 = DynamicParameter() {defaultValue=<rgba 1 1 0.01 1>, paramNames [TextureTile,WPOMult,WPOTile,WPOSpeed]}
  n4 = Add(A=Multiply(A=Subtract(A=@n1.XYZ, B=@n2), B=@n3.WPOTile), B=Multiply(A=Time(), B=@n3.WPOSpeed))
  n5 = ConstantBiasScale(None=Sine(None=ComponentMask(None=@n4) {b=true}))
  n6 = ScalarParameter() {defaultValue=1, parameterName=WPO_SineWave_Hardness}
  n7 = DynamicParameter() {defaultValue=<rgba 1 1 0.01 1>, paramNames [TextureTile,WPOMult,WPOTile,WPOSpeed]}
  n8 = ScalarParameter() {defaultValue=15, parameterName=WPO_SineWave_Intensity}
  Multiply(A=Multiply(A=Power(Base=@n5, Exp=@n6), B=Multiply(A=@n7.WPOMult, B=@n8)), B=VertexNormalWS())
  n9 = DynamicParameter() {parameterIndex=1, defaultValue=<rgba 1 0.1 0.01 1>, paramNames [HueShift,Emmisive,Param3,Param4]}
  n10 = ParticleColor()
  n11 = DynamicParameter() {parameterIndex=1, defaultValue=<rgba 1 0.1 0.01 1>, paramNames [HueShift,Emmisive,Param3,Param4]}
  n12 = Multiply(A=TextureCoordinate(), B=DynamicParameter() {defaultValue=<rgba 1 1 0.01 1>, paramNames [TextureTile,WPOMult,WPOTile,WPOSpeed]}.TextureTile)
  Multiply(
   A=@n9.Emmisive
   B=LinearInterpolate(
    A=@n10.RGB
    B=MaterialFunctionCall(Hue Shift Percentage=@n11.HueShift, Texture=@n10.RGB) {materialFunction=&HueShift}.Result
    Alpha=TextureSample(UVs=@n12) {texture=/Game/Environments/Asteria/Props/Holiday/Holiday_Presents_A/Textures/T_Holiday_WrappingPaper_A}.G
   )
  )
  ScalarParameter() {defaultValue=0.01, parameterName=WPO_SineWave_Tiling}
  ScalarParameter() {defaultValue=1, parameterName=WPO_SineWave_PanSpeed}
  ScalarParameter() {defaultValue=0.02, parameterName=Emmsive}
 ]
}
