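# Niagara one-shot fire burst for the front-end Cozy environment, apparently the 5000 XP variant.
# A brief flare sprite with light pops alongside a flurry of rising GPU embers.
# A launched Head sprite feeds short- and long-lived sparkle trails through particle reads, with curl noise, flickering alpha and speed-based color.
# A camera-facing Grid2D gas smoke/fire sim and its Head-fed SourceParticles emitter are present but disabled.
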NiagaraSystem /Game/Environments/FrontEnd/FX/Cozy/NS_Cozy_Fire_5000XP
dependencies {
 usedDataInterfaces [
  &NiagaraDataInterfaceCurve
  &NiagaraDataInterfaceVectorCurve
  &NiagaraDataInterfaceParticleRead
  &NiagaraDataInterfaceColorCurve
  /Script/Niagara.NiagaraDataInterfaceGrid2DCollection
  /Script/Niagara.NiagaraDataInterfaceRenderTarget2D
  &NiagaraDataInterfaceCamera
  &NiagaraDataInterfacePlatformSet
  /Script/Niagara.NiagaraDataInterfaceEmitterProperties
 ]
 usedDynamicInputs []
 usedModules [&SystemState]
}
emitters {
 CenterFlarePop {
  emitterUpdateScript [
   &EmitterState()
   &SpawnBurst_Instantaneous(Use Spawn Probability=false, UseLoopCountLimit=false)
  ]
  particleSpawnScript [
   &InitializeParticle(Lifetime=0.175, Sprite Size=<150 150>, Sprite Size Mode=::NewEnumerator4, UsePositionOffset=false, Write Material Random=true)
   &Color(Color=&MakeLinearColorFromVectorAndFloat("Vector (RGB)"=<5 2.8304 0.5873>, "Float (Alpha)"=&ClampFloat(Float=1)))
  ]
  particleUpdateScript [
   &ParticleState()
   &ScaleSpriteSize()
   &ScaleColor(
    Scale Alpha=&FloatFromCurve(FloatCurve=DataInterface() {propertyValues {Curve {keys [<at 0 0 RCIM_Cubic>,<at 0.1015 0.9974 RCIM_Cubic>,<at 1 0 RCIM_Cubic>]}}})
    Scale RGB=&VectorFromCurve(VectorCurve=DataInterface() {propertyValues {XCurve {keys [<at 0 1>,<at 1 1>]}, YCurve {keys [<at 0 1>,<at 1 1>]}, ZCurve {keys [<at 0 1>,<at 1 1>]}}})
    ScaleA=true
    ScaleRGB=true
   )
   &DynamicMaterialParameters(
    Index 0 Param 1=0.5
    Index 0 Param 2=1
    Index 0 Param 3=1
    Index 0 Param 4=1
    Module.Param0WriteEnabled=true
    Module.Param1WriteEnabled=true
    Module.Param2WriteEnabled=true
    Module.Param3WriteEnabled=true
   )
   &CameraOffset(Camera Offset Amount=20)
   &FN_Light_Attributes(Create Color Attribute=true, Create Exponent Attribute=true, Create Radius Attribute=true)
  ]
  rendererClasses [&NiagaraSpriteRendererProperties, &NiagaraLightRendererProperties]
  renderers [
   {rendererClass=&NiagaraSpriteRendererProperties}
   {rendererClass=&NiagaraLightRendererProperties, rendererIndex=1}
  ]
 }
 Embers_Flury {
  simTarget=GPUComputeSim
  emitterUpdateScript [
   &EmitterState()
   &SpawnBurst_Instantaneous(Spawn Count=10, Use Spawn Probability=false, UseLoopCountLimit=false)
   &SpawnRate(
    SpawnRate=&FloatFromCurve(
     FloatCurve=DataInterface() {propertyValues {Curve {keys [<at 0 100 RCIM_Cubic>,<at 0.25 0 RCIM_Cubic>]}}}
     CurveIndex=$Emitter.NormalizedLoopAge
    )
    Use Spawn Probability=false
   )
  ]
  particleSpawnScript [
   &InitializeParticle(
    Lifetime=&RandomRangeFloat(Minimum=0.25, Maximum=0.75)
    Sprite Size=&RandomRangeVector2D(Minimum=<1 2>, Maximum=<2 4>)
    Sprite Size Mode=::NewEnumerator4
    UsePositionOffset=false
    Write Material Random=true
   )
   &ShapeLocation(MaskSpawnGroup=false, Override Seed=false, Sphere Radius=2, UseOwnerScale=true)
   &AddVelocity(Velocity=&Multiply_VectorByFloat(Vector=$Particles.ShapeLocation.ShapeVector, Float=100)) {module=AddVelocity001}
   &AddVelocity(Velocity=&RandomRangeVector(Minimum=<0 0 25>, Maximum=<200 0 150>))
   &Color(Color=&MakeLinearColorFromVectorAndFloat("Vector (RGB)"=<2 1.2022 0.6482>, "Float (Alpha)"=&ClampFloat(Float=1)))
  ]
  particleUpdateScript [
   &ParticleState()
   &ScaleSpriteSize()
   &AccelerationForce(Acceleration=<0 0 350>)
   &Drag(Drag=3, Use Linear Drag=true, Use Rotational Drag=true)
   &SolveForcesAndVelocity(Clamp Velocity=false, Limit Acceleration=false)
   &ScaleSpriteSizeBySpeed(Max Scale Factor=<10 10>, Source Velocity:Vector3f=&Multiply_Vector(A=<0 0.01 1>, B=$Particles.Velocity))
   &ScaleColor(
    Scale Alpha=&FloatFromCurve(FloatCurve=DataInterface() {propertyValues {Curve {keys [{arriveTangent=11.7282, interpMode=RCIM_Cubic, time=0, value=0}, <at 0.2 1 RCIM_Cubic>, <at 1 0>]}}})
    Scale RGB=&VectorFromCurve(VectorCurve=DataInterface() {propertyValues {XCurve {keys [<at 0 4>,<at 1 1>]}, YCurve {keys [<at 0 1.3836>,<at 1 0.0572>]}, ZCurve {keys [<at 0 0>,<at 1 0>]}}})
    ScaleA=true
    ScaleRGB=true
   )
  ]
  rendererClasses [&NiagaraSpriteRendererProperties]
  renderers [{rendererClass=&NiagaraSpriteRendererProperties}]
 }
 Grid2D_Gas_SmokeFire {
  bEnabled=false
  simTarget=GPUComputeSim
  emitterSpawnScript [
   /SetVariables(Emitter.ResX=128, Emitter.ResY=128) {isSetParameters=true}
   /SetVariables(Emitter.SimGrid=DataInterface() {propertyValues {
      ClearBeforeNonIterationStage=true
      NumAttributes=0
      NumCellsMaxAxis=3
      NumCellsX=1
      NumCellsY=1
      OverrideBufferFormat=Float
      PreviewAttribute=None
      SetGridFromMaxAxis=false
      bOverrideFormat=false
      bPreviewGrid=false
      RenderTargetUserParameter {parameter {name=None, typeDefHandle {registeredTypeIndex=1}, varData []}}
      WorldBBoxSize=<1200 1200>
     }}) {isSetParameters=true}
   /SetVariables(Emitter.LightingGrid=DataInterface() {propertyValues {
      ClearBeforeNonIterationStage=true
      NumAttributes=3
      NumCellsMaxAxis=3
      NumCellsX=196
      NumCellsY=196
      OverrideBufferFormat=Float
      PreviewAttribute=None
      SetGridFromMaxAxis=false
      bOverrideFormat=false
      bPreviewGrid=false
      RenderTargetUserParameter {parameter {name=None, typeDefHandle {registeredTypeIndex=4}, varData []}}
      WorldBBoxSize=<100 100>
     }}) {isSetParameters=true}
   /SetVariables(Emitter.LightingGridDownsample=DataInterface() {propertyValues {
      ClearBeforeNonIterationStage=true
      NumAttributes=1
      NumCellsMaxAxis=3
      NumCellsX=64
      NumCellsY=64
      OverrideBufferFormat=Float
      PreviewAttribute=None
      SetGridFromMaxAxis=false
      bOverrideFormat=false
      bPreviewGrid=false
      RenderTargetUserParameter {parameter {name=None, typeDefHandle {registeredTypeIndex=4}, varData []}}
      WorldBBoxSize=<100 100>
     }}) {isSetParameters=true}
   /SetVariables() {isSetParameters=true}
   /NiagaraFluids/Modules/Grid2D/Grid2D_SetResolution(Grid:NiagaraDataInterfaceGrid2DCollection=$Emitter.SimGrid, NumCellsX=256, NumCellsY=256, WorldGridExtents:Vector2f=<200 200>)
   &Grid2D_SetResolution(
    Grid:NiagaraDataInterfaceGrid2DCollection=$Emitter.PressureGrid
    OtherGrid:NiagaraDataInterfaceGrid2DCollection=$Emitter.SimGrid
    Resolution Method:ENiagaraGrid2DResolution=::NewEnumerator3
    WorldGridExtents:Vector2f=$Emitter.Grid2D_SetResolution.WorldGridExtents
   ) {module=Grid2D_SetResolution001}
   &Grid2D_SetResolution(
    Grid:NiagaraDataInterfaceGrid2DCollection=$Emitter.LightingGridDownsample
    OtherGrid:NiagaraDataInterfaceGrid2DCollection=$Emitter.SimGrid
    Resolution Method:ENiagaraGrid2DResolution=::NewEnumerator3
    Resolution Mult=0.2
    WorldGridExtents:Vector2f=$Emitter.Grid2D_SetResolution.WorldGridExtents
   ) {module=Grid2D_SetResolution002}
   &Grid2D_SetResolution(
    Grid:NiagaraDataInterfaceGrid2DCollection=$Emitter.LightingGrid
    OtherGrid:NiagaraDataInterfaceGrid2DCollection=$Emitter.SimGrid
    Resolution Method:ENiagaraGrid2DResolution=::NewEnumerator3
    WorldGridExtents:Vector2f=$Emitter.Grid2D_SetResolution.WorldGridExtents
   ) {module=Grid2D_SetResolution003}
   /SetVariables(Emitter.dt=0.0167) {isSetParameters=true}
   /SetVariables(Emitter.g:Vector2f=<0 -9.8>) {isSetParameters=true}
   /SetVariables() {isSetParameters=true}
   /SetVariables(Emitter.density=1) {isSetParameters=true}
   /SetVariables(Emitter.ParticleSource=DataInterface() {propertyValues {EmitterBinding {bindingMode=Other, emitterName=SourceParticles}}}) {isSetParameters=true}
   /SetVariables() {isSetParameters=true}
   &self:SetLightingRT(LightingGrid:NiagaraDataInterfaceGrid2DCollection=$Emitter.LightingGrid, LightingRT:NiagaraDataInterfaceRenderTarget2D=$Emitter.LightingRT) {module=SetLightingRT}
   /NiagaraFluids/Modules/Grid2D/Grid2D_UpdateSimAttributes(Grid:NiagaraDataInterfaceGrid2DCollection=$Emitter.SimGrid, WorldGridSize:Vector2f=$Emitter.Grid2D_SetResolution.WorldGridExtents) {module=UpdateSimAttributes}
   /SetVariables() {isSetParameters=true}
   /SetVariables(Emitter.testCurve=DataInterface() {propertyValues {
      bExposeCurve=true
      AlphaCurve {keys []}
      BlueCurve {keys [<at 0.3032 1>,<at 0.6671 0>]}
      GreenCurve {keys [<at 0.3032 1>,<at 0.6671 0.0785>]}
      RedCurve {keys [<at 0.3032 1>,<at 0.6671 1>]}
     }}) {isSetParameters=true}
   /NiagaraFluids/Functions/General/Fluids_Gas_Manager(
    Density Color Curve=DataInterface() {propertyValues {
      ExposedName=Curve Density
      bExposeCurve=true
      AlphaCurve {keys [<at 0 0 RCIM_Cubic>,<at 0.0846 0>,<at 0.5224 1>,<at 0.8988 1>,<at 1 0>]}
      BlueCurve {keys [<at 0 0.2>,<at 0.5169 0>]}
      GreenCurve {keys [<at 0 0.1906>,<at 0.5169 0.6088>]}
      RedCurve {keys [<at 0 0.1885>,<at 0.5169 5>]}
     }}
    Depth Fade=25
    Dissipation Density=0.5
    Dissipation Temperature=0
    Dissipation Velocity=2
    EnvironmentColor:LinearColor=<rgba 0 0 0 1>
    Pressure Relaxation=1
    Radius Scale=0.25
    SunColor:LinearColor=<rgba 0 0 0 1>
    Temperature Bouyancy=-0.05
    Temperature Color Curve=DataInterface() {propertyValues {
      ExposedName=Curve Temperature
      bExposeCurve=true
      AlphaCurve {keys [<at 0.0198 0>,<at 0.1606 0>,<at 0.2245 0>,<at 0.8601 1>,<at 0.9326 0>]}
      BlueCurve {keys [<at 0.0638 0>,<at 0.3024 0>,<at 0.8019 0 RCIM_Cubic>,<at 0.9689 0>]}
      GreenCurve {keys [
        <at 0.0638 0>
        <at 0.3024 27.4248>
        {arriveTangent=0.7102, interpMode=RCIM_Cubic, leaveTangent=0.7102, tangentMode=RCTM_User, time=0.5996, value=0.1724}
        {arriveTangent=-0.4668, interpMode=RCIM_Cubic, leaveTangent=-0.4668, time=0.8064, value=0.8277}
        <at 0.9689 0>
       ]}
      RedCurve {keys [<at 0.0638 0>,<at 0.3024 250>,<at 0.7904 2.5789 RCIM_Cubic>,<at 0.9689 0>]}
     }}
    Temperature Maximum=2500
   )
   /NiagaraFluids/Modules/Grid2D/Grid2D_InitRT(Grid:NiagaraDataInterfaceGrid2DCollection=$Emitter.SimGrid, SimRT:NiagaraDataInterfaceRenderTarget2D=$Emitter.SimRT) {module=Grid2D_Gas_InitRT}
   &self:Grid2D_Gas_SmokeFire_0.NiagaraScratchPadContainer_0.ScratchModule_0(Grid:NiagaraDataInterfaceGrid2DCollection=$Emitter.SimGrid, PressureGrid:NiagaraDataInterfaceGrid2DCollection=$Emitter.PressureGrid) {module=ScratchModule}
  ]
  emitterUpdateScript [
   /NiagaraFN/Modules/Scalability/SignificantInstancesCull(Cull Response:ENiagaraExecutionStateManagement=::SleepAndLetParticlesFinish, First Frame Only=false)
   &EmitterState(Scalability Mode=::NewEnumerator1)
   /SetVariables(Emitter.dt:NiagaraFloat=$Engine.DeltaTime) {isSetParameters=true}
   /NiagaraFluids/Modules/Grid2D/Grid2D_CreateUnitToWorldTransform(
    Camera:NiagaraDataInterfaceCamera=$Emitter.CameraQuery
    Camera Facing=true
    Local Pivot:Vector2f=<0.5 0.05>
    Offset:NiagaraPosition=$Engine.Owner.Position
    Rotation:Quat4f=$Engine.Owner.Rotation
    WorldGridExtents:Vector2f=$Emitter.Grid2D_SetResolution.WorldGridExtents
   )
   /SetVariables(Emitter.Position:NiagaraPosition=&ConvertVectorToPosition(
     Input Position=$Emitter.Grid2D_CreateUnitToWorldTransform.LocalToWorldTranslation
     Conversion Operation:ENiagara_LWCConvertVecToPos=::NewEnumerator0
    )) {isSetParameters=true}
   /NiagaraFluids/Modules/Grid2D/Grid2D_SetBounds(
    CameraFacing:NiagaraBool=$Emitter.Grid2D_CreateUnitToWorldTransform.CameraFacing
    EmitterProperties=DataInterface() {propertyValues {EmitterBinding {bindingMode=Self, emitterName=None}}}
    LocalPivot:Vector2f=$Emitter.Grid2D_CreateUnitToWorldTransform.LocalPivot
    Rotate for Camera Facing=true
    WorldGridExtents:Vector2f=$Emitter.Grid2D_SetResolution.WorldGridExtents
   )
  ]
  rendererClasses [&NiagaraSpriteRendererProperties]
  renderers [
   {rendererClass=&NiagaraSpriteRendererProperties}
   {rendererClass=&NiagaraSpriteRendererProperties, rendererIndex=1}
  ]
 }
 Head {
  emitterUpdateScript [
   &EmitterState()
   &SpawnBurst_Instantaneous(Use Spawn Probability=false, UseLoopCountLimit=false)
  ]
  particleSpawnScript [
   &InitializeParticle(
    Lifetime=&RandomRangeFloat(Minimum=2.8, Maximum=3)
    Sprite Size=<75 25>
    Sprite Size Mode=::NewEnumerator4
    UsePositionOffset=false
    Write Material Random=true
   )
   &Color(Color=&MakeLinearColorFromVectorAndFloat("Vector (RGB)"=<1.25 1.25 1.25>, "Float (Alpha)"=&ClampFloat(Float=1)))
   &AddVelocity(Velocity=&RandomRangeVector(Minimum=<100 -50 100>, Maximum=<150 50 120>))
  ]
  particleUpdateScript [
   &ParticleState()
   &ScaleSpriteSize(Uniform Curve Sprite Scale=DataInterface() {propertyValues {Curve {keys [
        {arriveTangent=4.1785, interpMode=RCIM_Cubic, leaveTangent=4.1785, tangentMode=RCTM_User, time=0, value=0}
        {arriveTangent=1.1179, interpMode=RCIM_Cubic, leaveTangent=1.1179, tangentMode=RCTM_User, time=0.2042, value=0.6228}
        {arriveTangent=0.0224, interpMode=RCIM_Cubic, leaveTangent=0.0224, tangentMode=RCTM_User, time=1, value=1}
       ]}}}) {module=ScaleSpriteSize001}
   &Drag(Drag=2, Use Linear Drag=true, Use Rotational Drag=true)
   &AccelerationForce()
   &SolveForcesAndVelocity(Clamp Velocity=false, Limit Acceleration=false)
   &CameraOffset(Camera Offset Amount=25)
  ]
  rendererClasses [&NiagaraSpriteRendererProperties]
  renderers [{rendererClass=&NiagaraSpriteRendererProperties}]
 }
 SourceParticles {
  bEnabled=false
  emitterSpawnScript [&Grid2D_UpdateSimAttributes(Grid:NiagaraDataInterfaceGrid2DCollection=$Emitter.RasterizationGrid, WorldGridSize:Vector2f=$Emitter.Grid2D_SetResolution.WorldGridExtents) {module=UpdateSimAttributes}]
  emitterUpdateScript [
   &SpawnParticlesFromOtherEmitter(
    Attribute Reader=DataInterface() {propertyValues {EmitterBinding {bindingMode=Other, emitterName=Head}}}
    Maximum Age Spawn Threshold=0.5
    Minimum Age Spawn Threshold=0
    Spawn Rate=250
    Spawn Rate Per Particle Cap=250
    Use Maximum Age=true
    Use Minimum Age=true
   )
   &SignificantInstancesCull(Cull Response:ENiagaraExecutionStateManagement=::SleepAndLetParticlesFinish, First Frame Only=false)
   &EmitterState(
    Life Cycle Mode=::NewEnumerator1
    Loop Behavior=::NewEnumerator1
    Loop Duration Mode=::NewEnumerator1
    Scalability Mode=::NewEnumerator1
    UseLoopDelay=false
   )
   &SpawnBurst_Instantaneous(
    Spawn Count=&RandomRangeInt(Minimum=100, Maximum=100, Evaluation Type=::NewEnumerator1)
    Use Spawn Probability=false
   ) {enabled=false}
   &SpawnRate(SpawnRate=0, Use Spawn Probability=false)
  ]
  particleSpawnScript [
   /Niagara/Modules/Spawn/Initialization/InitializeParticle(
    Color=<rgba 1111 13.6009 0 1>
    Lifetime=&RandomRangeFloat(Minimum=0.5, Maximum=0.5)
    Mass=&RandomRangeFloat(Minimum=0.75, Maximum=1)
    Module.Write Color=true
    Module.Write Lifetime=true
    Module.Write Mass=true
    Module.Write Material Random=true
    Module.Write Position=false
    Module.Write Scale=false
    Module.Write SpriteRotation=false
    Module.Write SpriteSize=true
    Sprite Size=&Vector2DFromFloat(Value=&RandomRangeFloat(Minimum=1, Maximum=1))
   )
   &AddVelocityInCone(
    Cone Angle=&Cosine(Normalized Angle=$Emitter.Age, Period=0.4, Scale=0, Bias=88)
    Cone Axis=<0 0 1>
    Cone Axis Coordinate Space=::World
    Module.Use Velocity Falloff On Cone Axis=true
    Velocity Falloff Away From Cone Axis=0.333
    Velocity Strength=&RandomRangeFloat(Minimum=20, Maximum=50)
   )
   &ApplyInitialForces()
   &SampleParticlesFromOtherEmitter(Applied Velocity Scale=0.25, Velocity Sampling=::NewEnumerator0)
   &ShapeLocation(
    Box Size=<30 30 10>
    MaskSpawnGroup=false
    Offset=<25 0 0>
    Offset Mode=::NewEnumerator0
    Override Seed=false
    Shape Primitive=::NewEnumerator2
    Surface Only=false
    UseOwnerScale=true
   )
  ]
  particleUpdateScript [
   &ParticleState()
   &ScaleColor(
    Scale Alpha=&FloatFromCurve(FloatCurve=DataInterface() {propertyValues {Curve {keys [<at -0.0046 1.0113 RCIM_Cubic>,<at 1.0011 0.002 RCIM_Cubic>]}}})
    ScaleA=true
    ScaleRGB=true
   )
   &Collision(
    Age Colliding Particles=false
    Enable Rest State=false
    Friction=0
    Module.EnableMaxCollisionCount=false
    Module.Randomize Collision Normal Vector=false
    Restitution=0.7
   ) {enabled=false, module=Collision001}
   &CurlNoiseForce(Noise Frequency=250, Noise Strength=500, Pan Noise=false)
   &GravityForce(Gravity=&RandomRangeVector(Minimum=<0 0 0>, Maximum=<0 0 1>)) {module=GravityForce001}
   &SolveForcesAndVelocity(Clamp Velocity=false, Limit Acceleration=false)
   /NiagaraFluids/Modules/Particles/Fluids_Gas_Source(
    Density=0.1
    Divergence=2
    NoiseScale=5
    Radius:NiagaraFloat=&FloatFromCurve(
     FloatCurve=DataInterface() {propertyValues {Curve {keys [<at 0 2>,<at 0.2 0>,<at 1 5>]}}}
     Scale Curve=&RandomRangeFloat(Minimum=3, Maximum=5)
    )
    Temperature:NiagaraFloat=&FloatFromCurve(
     FloatCurve=DataInterface() {propertyValues {Curve {keys [<at 0 1.275>,<at 0.3 1.275>,<at 1 0>]}}}
     Scale Curve=&RandomRangeFloat(Minimum=500, Maximum=1100)
    )
    VelocityScale:NiagaraFloat=&RandomRangeFloat(Minimum=1, Maximum=2.5)
   )
   /SetVariables(Particles.SpriteSize=&Vector2DFromFloat(Value=&Multiply_Float(A=$Particles.fluids_source_radius, B=0.5))) {isSetParameters=true}
  ]
  rendererClasses [&NiagaraSpriteRendererProperties]
  renderers [{rendererClass=&NiagaraSpriteRendererProperties}]
 }
 Sparkle_Trail {
  emitterUpdateScript [
   &EmitterState()
   &SpawnParticlesFromOtherEmitter(
    Attribute Reader=DataInterface() {propertyValues {EmitterBinding {bindingMode=Other, emitterName=Head}}}
    Maximum Age Spawn Threshold=0.5
    Minimum Age Spawn Threshold=0
    Spawn Rate=200
    Spawn Rate Per Particle Cap=3000
    Use Maximum Age=true
    Use Minimum Age=true
   )
  ]
  particleSpawnScript [
   /Niagara/Modules/Spawn/Initialization/V2/InitializeParticle(
    Lifetime=&RandomRangeFloat(Minimum=0.5, Maximum=2)
    Sprite Rotation Angle=0.125
    Sprite Rotation Mode=::NewEnumerator1
    Sprite Size=&Vector2DFromFloat(Value=&RandomRangeFloat(Minimum=3, Maximum=8))
    Sprite Size Mode=::NewEnumerator4
    UsePositionOffset=false
    Write Material Random=true
   )
   &Color(Color=&MakeLinearColorFromVectorAndFloat(
     "Vector (RGB)"=<2 2 2>
     "Float (Alpha)"=&ClampFloat(Float=&RandomRangeFloat(Minimum=0.2, Maximum=0.5))
    ))
   &ShapeLocation(Box Size=<25 5 5>, MaskSpawnGroup=false, Override Seed=false, Shape Primitive=::NewEnumerator2, Surface Only=false, UseOwnerScale=true)
   &AddVelocity(
    Velocity=&Multiply_VectorByFloat(Vector=$Particles.ShapeLocation.ShapeVector, Float=5)
    Velocity Speed Scale=&Multiply_Float(
     A=1.5
     B=&FloatFromCurve(
      FloatCurve=DataInterface() {propertyValues {Curve {keys [{interpMode=RCIM_Cubic, tangentMode=RCTM_User, time=0, value=1}, {interpMode=RCIM_Cubic, tangentMode=RCTM_User, time=0.6, value=0}]}}}
      CurveIndex=$Emitter.NormalizedLoopAge
     )
    )
   )
   &SampleParticlesFromOtherEmitter(Applied Velocity Scale=0.25, Velocity Sampling=::NewEnumerator0)
  ]
  particleUpdateScript [
   &ParticleState()
   &ScaleSpriteSize(
    Uniform Curve Scale=&Waveform(
     Waveform Count=::NewEnumerator2
     "Waveform [1]"=::NewEnumerator0
     "[1] Phase Scale"=&RandomRangeFloat(Minimum=0.5, Maximum=1)
     "[1] Frequency"=&RandomRangeFloat(Minimum=2, Maximum=3)
     "[1] Amplitude Min/Max"=<0.8 1>
     "Waveform [2] Blend"=::NewEnumerator2
     "[2] Input Position"=0
     "[2] Phase"=$Emitter.Age
     "[2] Phase Scale"=0.25
     "[2] Compound Frequency"=2
     "[2] Frequency"=&RandomRangeFloat(Minimum=0, Maximum=1)
     "[2] Amplitude Min/Max"=<0.9 1>
     "[2] Scale"=1
     "[2] Bias"=0
     "[2] Exponent"=1
     "[2] Pulse Duration"=0.1
     "[2] Clip Triangle Min/Max"=<0 0>
     "[2] Random Seed"=0
     "[3] Average"=false
     "[3] Input Position"=0
     "[3] Phase"=$Emitter.Age
     "[3] Phase Scale"=&RandomRangeFloat(Minimum=0.7, Maximum=1)
     "[3] Frequency"=1
     "[3] Compound Frequency"=2
     "[3] Amplitude Min/Max"=<0.5 1>
     "[3] Scale"=1
     "[3] Bias"=0
     "[3] Exponent"=1
     "[3] Pulse Duration"=0.1
     "[3] Clip Triangle Min/Max"=<0 0>
     "[3] Random Seed"=0
    )
    Uniform Curve Sprite Scale=DataInterface() {propertyValues {Curve {keys [
        {arriveTangent=9.0725, interpMode=RCIM_Cubic, leaveTangent=9.0725, tangentMode=RCTM_User, time=0, value=0}
        {arriveTangent=2.276, interpMode=RCIM_Cubic, leaveTangent=2.276, tangentMode=RCTM_User, time=0.1408, value=0.9847}
        {arriveTangent=-4.0873, interpMode=RCIM_Cubic, leaveTangent=-4.0873, tangentMode=RCTM_User, time=0.5, value=0.65}
        {arriveTangent=-0.0405, interpMode=RCIM_Cubic, leaveTangent=-0.0405, tangentMode=RCTM_User, time=1, value=0}
       ]}}}
   )
   &CurlNoiseForce(
    Noise Frequency=25
    Noise Strength=&FloatFromCurve(
     FloatCurve=DataInterface() {propertyValues {Curve {keys [{interpMode=RCIM_Cubic, tangentMode=RCTM_User, time=0, value=1}, {interpMode=RCIM_Cubic, tangentMode=RCTM_User, time=1, value=0}]}}}
     Scale Curve=150
    )
    Pan Noise=true
    Pan Noise Field=<0 0 -1>
   )
   &Drag(
    Drag=&FloatFromCurve(
     FloatCurve=DataInterface() {propertyValues {Curve {keys [
         {interpMode=RCIM_Cubic, tangentMode=RCTM_User, time=0, value=0}
         {interpMode=RCIM_Cubic, tangentMode=RCTM_User, time=0.5, value=1}
         {tangentMode=RCTM_User, time=1, value=1}
        ]}}}
     Scale Curve=3
    )
    Use Linear Drag=true
    Use Rotational Drag=true
   )
   &GravityForce(Gravity=<0 0 -25>)
   &SolveForcesAndVelocity(Clamp Velocity=false, Limit Acceleration=false)
   &ScaleColor(
    Scale Alpha=&FloatFromCurve(
     FloatCurve=DataInterface() {propertyValues {Curve {keys [
         {arriveTangent=-0.9663, time=0, value=1}
         {arriveTangent=-1.9205, interpMode=RCIM_Cubic, leaveTangent=-1.9205, time=0.0699, value=0.0337}
         {arriveTangent=5.4323, interpMode=RCIM_Cubic, leaveTangent=5.4323, time=0.1481, value=0.7155}
         {arriveTangent=-4.1955, interpMode=RCIM_Cubic, leaveTangent=-4.1955, time=0.2398, value=0.9568}
         {arriveTangent=-1.2507, interpMode=RCIM_Cubic, leaveTangent=-1.2507, time=0.2759, value=0.1793}
         {arriveTangent=5.0466, interpMode=RCIM_Cubic, leaveTangent=5.0466, time=0.3301, value=0.8439}
         {arriveTangent=-7.2167, interpMode=RCIM_Cubic, leaveTangent=-7.2167, time=0.3992, value=0.8016}
         {arriveTangent=0.3527, interpMode=RCIM_Cubic, leaveTangent=0.3527, time=0.4429, value=0.03}
         {arriveTangent=0.446, interpMode=RCIM_Cubic, leaveTangent=0.446, time=0.5662, value=0.8604}
         {arriveTangent=-1.8021, interpMode=RCIM_Cubic, leaveTangent=-1.8021, time=0.6579, value=0.1259}
         {arriveTangent=-0.3681, interpMode=RCIM_Cubic, leaveTangent=-0.3681, time=0.7872, value=0.4621}
         <at 1 0>
        ]}}}
     Scale Curve=2
    )
    Scale RGB=&Multiply_Vector(
     A=&MakeVectorFromLinearColorRGB(Color=&ColorFromCurve(ColorCurve=DataInterface() {propertyValues {
         AlphaCurve {keys [<at 0 1 RCIM_Cubic>,<at 1 0 RCIM_Cubic>]}
         BlueCurve {keys [<at 0 0 RCIM_Cubic>,<at 1 0 RCIM_Cubic>]}
         GreenCurve {keys [<at 0 0.2016 RCIM_Cubic>,<at 1 0.53 RCIM_Cubic>]}
         RedCurve {keys [<at 0 1 RCIM_Cubic>,<at 1 1 RCIM_Cubic>]}
        }}))
     B=&VectorFromFloat(Value=5)
    )
    ScaleA=true
    ScaleRGB=true
   )
   /Niagara/Modules/Update/Color/ScaleColorBySpeed(A Scale Min=0.25, Max Speed Threshold=150, Min Speed Threshold=10, RGB Scale Max=5, RGB Scale Min=1)
  ]
  rendererClasses [&NiagaraSpriteRendererProperties]
  renderers [{rendererClass=&NiagaraSpriteRendererProperties}]
 }
 Sparkle_TrailLong {
  emitterUpdateScript [
   &EmitterState()
   &SpawnParticlesFromOtherEmitter(
    Attribute Reader=DataInterface() {propertyValues {EmitterBinding {bindingMode=Other, emitterName=Head}}}
    Maximum Age Spawn Threshold=0.5
    Minimum Age Spawn Threshold=0
    Spawn Rate=200
    Spawn Rate Per Particle Cap=3000
    Use Maximum Age=true
    Use Minimum Age=true
   )
  ]
  particleSpawnScript [
   &InitializeParticle(
    Lifetime=&RandomRangeFloat(Minimum=4, Maximum=8)
    Sprite Rotation Angle=0.125
    Sprite Rotation Mode=::NewEnumerator1
    Sprite Size=&Vector2DFromFloat(Value=&RandomRangeFloat(Minimum=3, Maximum=10))
    Sprite Size Mode=::NewEnumerator4
    UsePositionOffset=false
    Write Material Random=true
   )
   &Color(Color=&MakeLinearColorFromVectorAndFloat(
     "Vector (RGB)"=<2 2 2>
     "Float (Alpha)"=&ClampFloat(Float=&RandomRangeFloat(Minimum=0.2, Maximum=0.5))
    ))
   &ShapeLocation(Box Size=<25 5 5>, MaskSpawnGroup=false, Override Seed=false, Shape Primitive=::NewEnumerator2, Surface Only=false, UseOwnerScale=true)
   &AddVelocity(
    Velocity=&Multiply_VectorByFloat(Vector=$Particles.ShapeLocation.ShapeVector, Float=5)
    Velocity Speed Scale=&Multiply_Float(
     A=1.5
     B=&FloatFromCurve(
      FloatCurve=DataInterface() {propertyValues {Curve {keys [{interpMode=RCIM_Cubic, tangentMode=RCTM_User, time=0, value=1}, {interpMode=RCIM_Cubic, tangentMode=RCTM_User, time=0.6, value=0}]}}}
      CurveIndex=$Emitter.NormalizedLoopAge
     )
    )
   )
   &SampleParticlesFromOtherEmitter(Applied Velocity Scale=0.25, Velocity Sampling=::NewEnumerator0)
  ]
  particleUpdateScript [
   &ParticleState()
   &ScaleSpriteSize(
    Uniform Curve Scale=&Waveform(
     Waveform Count=::NewEnumerator2
     "Waveform [1]"=::NewEnumerator0
     "[1] Phase Scale"=&RandomRangeFloat(Minimum=0.5, Maximum=1)
     "[1] Frequency"=&RandomRangeFloat(Minimum=2, Maximum=3)
     "[1] Amplitude Min/Max"=<0.8 1>
     "Waveform [2] Blend"=::NewEnumerator2
     "[2] Input Position"=0
     "[2] Phase"=$Emitter.Age
     "[2] Phase Scale"=0.25
     "[2] Compound Frequency"=2
     "[2] Frequency"=&RandomRangeFloat(Minimum=0, Maximum=1)
     "[2] Amplitude Min/Max"=<0.9 1>
     "[2] Scale"=1
     "[2] Bias"=0
     "[2] Exponent"=1
     "[2] Pulse Duration"=0.1
     "[2] Clip Triangle Min/Max"=<0 0>
     "[2] Random Seed"=0
     "[3] Average"=false
     "[3] Input Position"=0
     "[3] Phase"=$Emitter.Age
     "[3] Phase Scale"=&RandomRangeFloat(Minimum=0.7, Maximum=1)
     "[3] Frequency"=1
     "[3] Compound Frequency"=2
     "[3] Amplitude Min/Max"=<0.5 1>
     "[3] Scale"=1
     "[3] Bias"=0
     "[3] Exponent"=1
     "[3] Pulse Duration"=0.1
     "[3] Clip Triangle Min/Max"=<0 0>
     "[3] Random Seed"=0
    )
    Uniform Curve Sprite Scale=DataInterface() {propertyValues {Curve {keys [
        {arriveTangent=9.0725, interpMode=RCIM_Cubic, leaveTangent=9.0725, tangentMode=RCTM_User, time=0, value=0}
        {arriveTangent=2.276, interpMode=RCIM_Cubic, leaveTangent=2.276, tangentMode=RCTM_User, time=0.1408, value=0.9847}
        {arriveTangent=-4.0873, interpMode=RCIM_Cubic, leaveTangent=-4.0873, tangentMode=RCTM_User, time=0.5, value=0.65}
        {arriveTangent=-0.0405, interpMode=RCIM_Cubic, leaveTangent=-0.0405, tangentMode=RCTM_User, time=1, value=0}
       ]}}}
   )
   &CurlNoiseForce(
    Noise Frequency=25
    Noise Strength=&FloatFromCurve(
     FloatCurve=DataInterface() {propertyValues {Curve {keys [{interpMode=RCIM_Cubic, tangentMode=RCTM_User, time=0, value=1}, {interpMode=RCIM_Cubic, tangentMode=RCTM_User, time=0.25, value=0}]}}}
     Scale Curve=150
    )
    Pan Noise=true
    Pan Noise Field=<0 0 -1>
   )
   &Drag(
    Drag=&FloatFromCurve(
     FloatCurve=DataInterface() {propertyValues {Curve {keys [
         {interpMode=RCIM_Cubic, tangentMode=RCTM_User, time=0, value=0}
         {interpMode=RCIM_Cubic, tangentMode=RCTM_User, time=0.5, value=1}
         {tangentMode=RCTM_User, time=1, value=1}
        ]}}}
     Scale Curve=3
    )
    Use Linear Drag=true
    Use Rotational Drag=true
   )
   &GravityForce(Gravity=<0 0 -25>)
   &SolveForcesAndVelocity(Clamp Velocity=false, Limit Acceleration=false)
   &ScaleColor(
    Scale Alpha=&FloatFromCurve(
     FloatCurve=DataInterface() {propertyValues {Curve {keys [
         {arriveTangent=-0.9663, time=0, value=1}
         {arriveTangent=-1.9205, interpMode=RCIM_Cubic, leaveTangent=-1.9205, time=0.0699, value=0.0337}
         {arriveTangent=5.4323, interpMode=RCIM_Cubic, leaveTangent=5.4323, time=0.1481, value=0.7155}
         {arriveTangent=-4.1955, interpMode=RCIM_Cubic, leaveTangent=-4.1955, time=0.2398, value=0.9568}
         {arriveTangent=-1.2507, interpMode=RCIM_Cubic, leaveTangent=-1.2507, time=0.2759, value=0.1793}
         {arriveTangent=5.0466, interpMode=RCIM_Cubic, leaveTangent=5.0466, time=0.3301, value=0.8439}
         {arriveTangent=-7.2167, interpMode=RCIM_Cubic, leaveTangent=-7.2167, time=0.3992, value=0.8016}
         {arriveTangent=0.3527, interpMode=RCIM_Cubic, leaveTangent=0.3527, time=0.4429, value=0.03}
         {arriveTangent=0.446, interpMode=RCIM_Cubic, leaveTangent=0.446, time=0.5662, value=0.8604}
         {arriveTangent=-1.8021, interpMode=RCIM_Cubic, leaveTangent=-1.8021, time=0.6579, value=0.1259}
         {arriveTangent=-0.3681, interpMode=RCIM_Cubic, leaveTangent=-0.3681, time=0.7872, value=0.4621}
         <at 1 0>
        ]}}}
     Scale Curve=2
    )
    Scale RGB=&Multiply_Vector(
     A=&MakeVectorFromLinearColorRGB(Color=&ColorFromCurve(ColorCurve=DataInterface() {propertyValues {
         AlphaCurve {keys [<at 0 1 RCIM_Cubic>,<at 1 0 RCIM_Cubic>]}
         BlueCurve {keys [<at 0 0 RCIM_Cubic>,<at 0.3453 0>,<at 1 0 RCIM_Cubic>]}
         GreenCurve {keys [<at 0 0.2016 RCIM_Cubic>,<at 0.3453 0.04>,<at 1 0.53 RCIM_Cubic>]}
         RedCurve {keys [<at 0 1 RCIM_Cubic>,<at 0.3453 1>,<at 1 1 RCIM_Cubic>]}
        }}))
     B=&Multiply_VectorByFloat(Vector=<1 1 1>, Float=5)
    )
    ScaleA=true
    ScaleRGB=true
   )
   &ScaleColorBySpeed(A Scale Min=0.25, Max Speed Threshold=150, Min Speed Threshold=10, RGB Scale Max=5, RGB Scale Min=1)
  ]
  rendererClasses [&NiagaraSpriteRendererProperties]
  renderers [{rendererClass=&NiagaraSpriteRendererProperties}]
 }
}
summary {
 systemName=NS_Cozy_Fire_5000XP
 emitters [
  {emitterName=Embers_Flury, simTarget=GPUComputeSim, rendererClasses [&NiagaraSpriteRendererProperties]}
  {
   emitterName=CenterFlarePop
   rendererClasses [&NiagaraSpriteRendererProperties, &NiagaraLightRendererProperties]
  }
  {emitterName=Head, rendererClasses [&NiagaraSpriteRendererProperties]}
  {emitterName=Sparkle_Trail, rendererClasses [&NiagaraSpriteRendererProperties]}
  {emitterName=Sparkle_TrailLong, rendererClasses [&NiagaraSpriteRendererProperties]}
  {bEnabled=false, emitterName=Grid2D_Gas_SmokeFire, simTarget=GPUComputeSim, rendererClasses [&NiagaraSpriteRendererProperties]}
  {bEnabled=false, emitterName=SourceParticles, rendererClasses [&NiagaraSpriteRendererProperties]}
 ]
 userVariables []
}
