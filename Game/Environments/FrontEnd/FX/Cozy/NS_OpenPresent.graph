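# Niagara one-emitter burst of mesh particles for opening a present in the Cozy front-end environment.
# Particles spawn together on a box shape's surface with randomized scale, orientation and material random, tinted by the User.TestColor parameter.
# Initial velocity, gravity, strong curl noise and drag scatter them as they spin.
# A curve shrinks meshes to zero late in life, while dynamic material parameters feed fixed values to the material.
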NiagaraSystem /Game/Environments/FrontEnd/FX/Cozy/NS_OpenPresent
dependencies {
 usedDataInterfaces [&NiagaraDataInterfaceMeshRendererInfo, &NiagaraDataInterfaceCurve]
 usedDynamicInputs []
 usedModules [&SystemState]
}
emitters {SimpleSpriteBurst {
  emitterUpdateScript [
   &EmitterState(Life Cycle Mode=::NewEnumerator1, Loop Behavior=::NewEnumerator1, Loop Duration=2, UseLoopDelay=false)
   &SpawnBurst_Instantaneous(Spawn Count=40, Use Spawn Probability=false, UseLoopCountLimit=false)
  ]
  particleSpawnScript [
   /SetVariables(Particles.TestColor:LinearColor=$User.TestColor) {isSetParameters=true}
   &InitializeParticle(
    Color=$Particles.TestColor
    Lifetime=&RandomRangeFloat(Minimum=2, Maximum=3)
    Mesh Renderer Array Visibility Mode=::NewEnumerator3
    Mesh Renderer Info=DataInterface() {propertyValues {MeshRenderer=&self:SimpleSpriteBurst.NiagaraMeshRendererProperties_0}}
    Mesh Renderer Random Seed:NiagaraInt32=&RandomRangeInt(Minimum=0, Maximum=8)
    Mesh Scale Mode=::NewEnumerator1
    Mesh Uniform Scale Max=0.5
    Mesh Uniform Scale Min=0.1
    Sprite Size Mode=::NewEnumerator3
    Uniform Sprite Size=50
    UsePositionOffset=false
    Write Material Random=true
   )
   &ShapeLocation(Box Size=<50 50 50>, MaskSpawnGroup=false, Override Seed=false, Shape Primitive=::NewEnumerator2, Surface Only=true, UseOwnerScale=true)
   &AddVelocity(Module.Use Falloff Distance=false, Override Seed=false, Velocity Mode=::NewEnumerator1, Velocity Speed=300)
   &InitialMeshOrientation(
    Mesh Orientation Mode=::NewEnumerator2
    Orientation Vector=$Particles.ShapeLocation.ShapeNormal
    Rotation=&RandomRangeVector(Minimum=<0 0 0>, Maximum=<1 1 1>)
    Use Rotation Vector=true
   )
  ]
  particleUpdateScript [
   &ParticleState()
   &ScaleMeshSize(Scale Factor=&VectorFromFloat(Value=&FloatFromCurve(FloatCurve=DataInterface() {propertyValues {Curve {keys [<at 0.8 1>,<at 1 0>]}}})))
   &WindForce(Scale=0, Wind Speed=<100 0 100>)
   &GravityForce(Gravity=<0 0 -600>)
   &CurlNoiseForce(Noise Frequency=&RandomRangeFloat(Minimum=10, Maximum=25), Noise Strength=800, Pan Noise=false)
   &AerodynamicDrag(Aerodynamic Drag=0.04, Use Linear Drag=true, Use Rotational Drag=true)
   &SolveForcesAndVelocity(Clamp Velocity=false, Limit Acceleration=false)
   &AlignSpriteToMeshOrientation(SetAlignment=true, SetFacing=true)
   &UpdateMeshOrientation(Rotation Rate=&RandomRangeFloat(Minimum=1, Maximum=1.5), Rotation Vector=<0 0 1>)
   &DynamicMaterialParameters(
    Index 0 Param 1=0.5
    Index 0 Param 2=0.5
    Index 0 Param 3=0.05
    Index 0 Param 4=1
    Index 1 Param 1=0.15
    Index 1 Param 2=0.1
    Module.Param0WriteEnabled=true
    Module.Param0WriteEnabled1=true
    Module.Param1WriteEnabled=true
    Module.Param1WriteEnabled1=true
    Module.Param2WriteEnabled=true
    Module.Param2WriteEnabled1=true
    Module.Param3WriteEnabled=true
    Module.Param3WriteEnabled1=true
    Write Parameter Index 1=true
   )
   &ScaleColor(ScaleA=true, ScaleRGB=true) {enabled=false}
  ]
  rendererClasses [&NiagaraMeshRendererProperties]
  renderers [{rendererClass=&NiagaraMeshRendererProperties, rendererIndex=0}]
 }}
summary {
 systemName=NS_OpenPresent
 emitters [{emitterName=SimpleSpriteBurst, rendererClasses [&NiagaraMeshRendererProperties]}]
 userVariables [User.TestColor {type=LinearColor, description="", defaultValue {struct=LinearColor, value=<rgba 1 0 0 1>}}]
}
